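; BP_ShowcaseInterface declares the function contract for the asset showcase art tool, kept under NeverCook.
; CurrentLOD carries a vertex count and LOD label, while SetBurnIn passes a burn-in reference.
; ResolveBinding takes a Movie Scene binding proxy, and SetSequencerTOD takes a Fort Day Sequence actor with start/stop and min/max values for time of day.
; The functions hold only signatures and default returns, with no logic.
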(fn CurrentLOD
  (params (Vertices :type Integer) (CurLOD :type String))
  (return))

(fn ResolveBinding
  (params (binding :type Movie%20Scene%20Binding%20Proxy%20Structure))
  (return 0))

(fn SetBurnIn
  (params (BurninRef :type BP%20Showcase%20Burn%20In%20Rel%20Object%20Reference)))

(fn SetSequencerTOD
  (params (TODActor :type Fort%20Day%20Sequence%20Actor%20Object%20Reference) (fStart :type Integer) (fStop :type Integer) (minVal :type Float%20%28double-precision%29) (maxVal :type Float%20%28double-precision%29))
  (return false))
